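# Niagara system for the NudgeHonk broom's green glow, built from two sprite emitters.
# Airglow slowly spawns large, drifting sprites whose size and alpha follow life curves, with alpha scaled by time of day.
# It also writes star color/size and light attributes, though only a sprite renderer is listed.
# Sparkle rapidly emits small motes from a cone, flickering through randomized waveform alpha as they dim and slow under drag and gravity.
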NiagaraSystem /FruitFire_NudgeHonk/Assets/VFX/NS_NudgeHonk_BroomGlow_01
dependencies {
 usedDataInterfaces [&NiagaraDataInterfaceCurve]
 usedDynamicInputs [&Boolean_NotOperation]
 usedModules [
  &SystemState
  &self:SystemSpawnScript.SystemScriptSource.SystemScriptGraph.NiagaraNodeAssignment_0.SetVariables_#1
 ]
}
emitters {
 Airglow {
  emitterUpdateScript [
   &EmitterState()
   &SpawnRate(SpawnRate=2, Use Spawn Probability=false)
  ]
  particleSpawnScript [
   &InitializeParticle(
    Color=&RandomRangeLinearColor(Minimum=<rgba 0 1 0 0.5>, Maximum=<rgba 0.2 1 0.2 0.6>)
    Lifetime=2.35
    Sprite Rotation Angle=&RandomRangeFloat(Minimum=-1, Maximum=1)
    Sprite Rotation Mode=::NewEnumerator1
    Sprite Size Mode=::NewEnumerator3
    Uniform Sprite Size=&RandomRangeFloat(Minimum=110, Maximum=140)
    UsePositionOffset=false
    Write Material Random=true
   )
   &CameraOffset(Camera Offset Amount=15)
   &SystemLocation(MaskSpawnGroup=false, Offset=<0 0 -20>)
   &AddVelocity(Velocity=<3 0 -30>)
   /SetVariables(Particles.CameraOffset_Airglow=15) {isSetParameters=true}
  ]
  particleUpdateScript [
   &ParticleState()
   &ScaleColor(
    Scale Alpha=&Multiply_Float(
     A=&Multiply_Float(
      A=&FloatFromCurve(FloatCurve=DataInterface() {propertyValues {Curve {
          defaultValue=1
          keys [
           {arriveTangent=5.4319, interpMode=RCIM_Cubic, leaveTangent=5.4319, tangentMode=RCTM_User, time=0, value=0}
           {arriveTangent=1.1599, interpMode=RCIM_Cubic, leaveTangent=1.1599, tangentMode=RCTM_User, time=0.2391, value=0.8641}
           {arriveTangent=-1.2592, interpMode=RCIM_Cubic, leaveTangent=-1.2592, tangentMode=RCTM_User, time=0.8515, value=0.8591}
           {arriveTangent=0.0434, tangentMode=RCTM_User, time=1, value=0}
          ]
         }}})
      B=&FNTimeOfDayScalar(Night Scale=0.75, Morning Scale=0.9, Evening Scale=0.9)
     )
     B=1
    )
    ScaleA=true
    ScaleRGB=false
   )
   &ScaleSpriteSize(
    Scale Sprite Size Mode=::NewEnumerator0
    Uniform Scale Factor=&FloatFromCurve(FloatCurve=DataInterface() {propertyValues {Curve {keys [
         {arriveTangent=5.5076, interpMode=RCIM_Cubic, leaveTangent=5.5076, tangentMode=RCTM_User, time=0.0024, value=0.4086}
         {arriveTangent=0.8411, interpMode=RCIM_Cubic, leaveTangent=0.8411, tangentMode=RCTM_User, time=0.3896, value=0.906}
         {arriveTangent=-0.1314, interpMode=RCIM_Cubic, leaveTangent=-0.1314, tangentMode=RCTM_User, time=0.8033, value=1.407}
         {arriveTangent=-3.5952, tangentMode=RCTM_User, time=0.95, value=0.9}
         <at 1 0.25>
        ]}}})
   )
   /SetVariables(Particles.ColorStars:LinearColor=&MultiplyLinearColor(
     Linear Color=<rgba 0.15 1 0.3625 0.7>
     Scale Factor=&MakeFloat4FromVectorAndFloat(
      "Vector (RGB)"=<1 1 1>
      "Float (Alpha)"=&Multiply_Float(
       A=&FloatFromCurve(FloatCurve=DataInterface() {propertyValues {Curve {
           defaultValue=1
           keys [
            {arriveTangent=5.4319, interpMode=RCIM_Cubic, leaveTangent=5.4319, tangentMode=RCTM_User, time=0, value=0}
            {arriveTangent=1.1599, interpMode=RCIM_Cubic, leaveTangent=1.1599, tangentMode=RCTM_User, time=0.2391, value=0.8641}
            {arriveTangent=-1.2592, interpMode=RCIM_Cubic, leaveTangent=-1.2592, tangentMode=RCTM_User, time=0.8515, value=0.8591}
            {arriveTangent=0.0434, tangentMode=RCTM_User, time=1, value=0}
           ]
          }}})
       B=&FNTimeOfDayScalar(Night Scale=0.75, Morning Scale=0.9, Evening Scale=0.9)
      )
     )
    )) {isSetParameters=true}
   /SetVariables(Particles.StarsSize:Vector2f=&Multiply_Vector2D_ByFloat(
     Float=&FloatFromCurve(FloatCurve=DataInterface() {propertyValues {Curve {keys [
          {arriveTangent=5.5076, interpMode=RCIM_Cubic, leaveTangent=5.5076, tangentMode=RCTM_User, time=0.0024, value=0.4086}
          {arriveTangent=0.8411, interpMode=RCIM_Cubic, leaveTangent=0.8411, tangentMode=RCTM_User, time=0.3896, value=0.906}
          {arriveTangent=-0.1314, interpMode=RCIM_Cubic, leaveTangent=-0.1314, tangentMode=RCTM_User, time=0.8033, value=1.407}
          {arriveTangent=-3.5952, tangentMode=RCTM_User, time=0.95, value=0.9}
          <at 1 0.25>
         ]}}})
     Vector2D=$Particles.SpriteSize
    )) {isSetParameters=true}
   &FN_Light_Attributes(Color=<rgba 0 10 0 1>, Create Color Attribute=true, Create Exponent Attribute=true, Create Radius Attribute=true, Exponent=1, Radius=60)
   &SolveForcesAndVelocity(Clamp Velocity=false, Limit Acceleration=false)
  ]
  rendererClasses [&NiagaraSpriteRendererProperties]
  renderers [{rendererClass=&NiagaraSpriteRendererProperties}]
 }
 Sparkle {
  emitterUpdateScript [
   &EmitterState()
   &SpawnRate(SpawnRate=32, Use Spawn Probability=false)
  ]
  particleSpawnScript [
   &InitializeParticle(
    Color=&RandomRangeLinearColor(Minimum=<rgba 0 1 0 1>, Maximum=<rgba 0.45 1.5 0.8 1>)
    Lifetime Max=1
    Lifetime Min=0.5
    Lifetime Mode=::NewEnumerator1
    Sprite Rotation Angle=0.125
    Sprite Rotation Mode=::NewEnumerator1
    Sprite Size Mode=::NewEnumerator1
    Uniform Sprite Size Max=16
    Uniform Sprite Size Min=12
    UsePositionOffset=false
    Write Material Random=true
   )
   &ShapeLocation(
    Cone Inner Angle=15
    Cone Length=70
    Flatten Endcaps=true
    MaskSpawnGroup=false
    Offset=<-1 0 -5>
    Offset Mode=::NewEnumerator0
    Override Seed=false
    Rotation Axis=<0.1 0 -1>
    Rotation Mode=::NewEnumerator1
    Shape Primitive=::NewEnumerator5
    UseOwnerScale=true
   )
   &AddVelocity(
    Module.Use Falloff Distance=false
    Origin Offset=<0 0 -45>
    Override Seed=false
    Velocity Mode=::NewEnumerator1
    Velocity Speed=&RandomRangeFloat(Minimum=25, Maximum=75)
   ) {module=AddVelocity001}
   &AddVelocity()
   &CameraOffset(Camera Offset Amount=15)
  ]
  particleUpdateScript [
   &ParticleState()
   &ScaleColor(
    Scale Alpha=&FloatFromCurve(
     FloatCurve=DataInterface() {propertyValues {Curve {keys [
         {tangentMode=RCTM_User, time=0, value=1}
         {arriveTangent=-0.0405, interpMode=RCIM_Cubic, leaveTangent=-0.0405, tangentMode=RCTM_User, time=0.7009, value=1}
         {arriveTangent=-8.0089, interpMode=RCIM_Cubic, leaveTangent=-8.0089, tangentMode=RCTM_User, time=1, value=0}
        ]}}}
     Scale Curve=&Waveform(
      Global Phase=&Add_Float(A=$Particles.Age, B=&RandomRangeFloat(Minimum=0, Maximum=1))
      "Waveform [1]"=::NewEnumerator0
      "[1] Frequency"=&RandomRangeFloat(Minimum=3.618, Maximum=11.08)
      "[1] Amplitude Min/Max"=<0 1>
     )
    )
    Scale RGB=&VectorFromFloat(Value=&Lerp_Float(
      A=0.24
      B=0.9
      Alpha=&FloatFromCurve(FloatCurve=DataInterface() {propertyValues {Curve {defaultValue=0, keys [{tangentMode=RCTM_User, time=0, value=1}, {tangentMode=RCTM_User, time=1, value=0}]}}})
     ))
    ScaleA=true
    ScaleRGB=true
   )
   &ScaleSpriteSize(Uniform Curve Sprite Scale=DataInterface() {propertyValues {Curve {keys [
        {interpMode=RCIM_Cubic, tangentMode=RCTM_User, time=0, value=0}
        {interpMode=RCIM_Cubic, tangentMode=RCTM_User, time=0.5, value=1}
        {tangentMode=RCTM_User, time=1, value=0}
       ]}}})
   &GravityForce(Gravity=&RandomRangeVector(Minimum=<0 0 -10>, Maximum=<1 1 -50>))
   &Drag(Drag=10, Use Linear Drag=true, Use Rotational Drag=false)
   &SolveForcesAndVelocity(Clamp Velocity=false, Limit Acceleration=false)
  ]
  rendererClasses [&NiagaraSpriteRendererProperties]
  renderers [{rendererClass=&NiagaraSpriteRendererProperties}]
 }
}
summary {
 systemName=NS_NudgeHonk_BroomGlow_01
 emitters [
  {emitterName=Airglow, rendererClasses [&NiagaraSpriteRendererProperties]}
  {emitterName=Sparkle, rendererClasses [&NiagaraSpriteRendererProperties]}
 ]
 userVariables [
  User.bIsFrontEnd {type=NiagaraBool}
  User.bIsFrontEndPreview {type=NiagaraBool}
  User.bisHandHidden {type=NiagaraBool}
  User.bKillReaction {type=NiagaraBool}
  User.IsSkydiving {type=NiagaraBool}
 ]
}
