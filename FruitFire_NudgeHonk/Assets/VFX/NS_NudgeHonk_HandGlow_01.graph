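# NS_NudgeHonk_HandGlow_01 is a Niagara system that puts a persistent green glow on a skeletal mesh's two hands.
# One delayed burst spawns two particles, partitioned to separate bones or sockets with mirrored offsets and opposite spins.
# They pop in with a size overshoot and outlive their lifetime, feeding two sprite renderers and a light.
# Sprite alpha scales by time of day; a camera-proximity ADS alpha step is disabled.
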NiagaraSystem /FruitFire_NudgeHonk/Assets/VFX/NS_NudgeHonk_HandGlow_01
dependencies {
 usedDataInterfaces [&NiagaraDataInterfaceCamera, &NiagaraDataInterfaceCurve, &NiagaraDataInterfaceSkeletalMesh]
 usedDynamicInputs [&Boolean_NotOperation]
 usedModules [
  &self:SystemSpawnScript.SystemScriptSource.SystemScriptGraph.NiagaraNodeAssignment_1.SetVariables_#1
  &SystemState
  &self:SystemSpawnScript.SystemScriptSource.SystemScriptGraph.NiagaraNodeAssignment_0.SetVariables_#2
 ]
}
emitters {Hand_Glow_Loop {
  emitterUpdateScript [
   &EmitterState(Life Cycle Mode=::NewEnumerator1, Loop Behavior=::NewEnumerator1, Loop Duration Mode=::NewEnumerator1, UseLoopDelay=false)
   &SpawnBurst_Instantaneous(Spawn Count=2, Spawn Time=0.3, Use Spawn Probability=false, UseLoopCountLimit=true)
  ]
  particleSpawnScript [
   &InitializeParticle(
    Color=<rgba 0 1.25 0 0.7>
    Lifetime=0.5
    Sprite Rotation Angle=&RandomRangeFloat(Minimum=-1, Maximum=1)
    Sprite Rotation Mode=::NewEnumerator1
    Sprite Size Mode=::NewEnumerator3
    Uniform Sprite Size=60
    UsePositionOffset=false
    Write Material Random=true
   )
   &PartitionParticles()
   &CameraOffset(Camera Offset Amount=25)
   /SetVariables(Particles.CameraOffsetSmoke=10) {isSetParameters=true}
  ]
  particleUpdateScript [
   /SetVariables(Particles.ADSAlpha:NiagaraFloat=&MakeCustomFloatFromBool(
     Bool=&SetBoolByFloatComparison(
      A=&VectorLength(Vector=&Subtract_Vector(
        A=&ReturnCameraPosition(Coordinate Space:ENiagaraCoordinateSpace=::World, Camera Query:NiagaraDataInterfaceCamera=DataInterface())
        B=Linked() {linkedVariable {name=Particles.Position, type=NiagaraPosition}}
       ))
      B=80
     )
     True Float=1
     False Float=0
    )) {enabled=false, isSetParameters=true}
   &ParticleState(Kill Particles When Lifetime Has Elapsed=false, Let Infinitely Lived Particles Die When Emitter Deactivates=true)
   &ScaleSpriteSize(Uniform Curve Sprite Scale=DataInterface() {propertyValues {Curve {
       defaultValue=0
       keys [
        {arriveTangent=18.8589, interpMode=RCIM_Cubic, leaveTangent=18.8589, tangentMode=RCTM_User, time=0, value=0}
        {arriveTangent=-0.1087, interpMode=RCIM_Cubic, leaveTangent=-0.1087, tangentMode=RCTM_User, time=0.2, value=1.4454}
        {arriveTangent=-0.0112, interpMode=RCIM_Cubic, leaveTangent=-0.0112, tangentMode=RCTM_User, time=0.55, value=0.9754}
        {arriveTangent=-0.0405, interpMode=RCIM_Cubic, leaveTangent=-0.0405, tangentMode=RCTM_User, time=1, value=1}
       ]
      }}})
   &SkeletalMeshLocation(
    Bone / Socket Index=$Particles.PartitionParticles.Partition
    Bone Sampling Mode=::NewEnumerator4
    MaskSpawnGroup=false
    Sampled Position Offset=&MaskVectorBySpawnGroup(Masked Vector=<8 -2 0>, Passthrough Vector=<-8 2 0>, Spawn Group Mask=$Particles.PartitionParticles.Partition)
    Skeletal Mesh=$System.Skeletal Mesh
   )
   &ScaleColor(
    Scale Alpha=&Multiply_Float(A=1, B=&FNTimeOfDayScalar(Night Scale=0.4, Morning Scale=0.7, Day Scale=0.9, Evening Scale=0.8))
    ScaleA=true
    ScaleRGB=false
   )
   &SpriteRotationRate(Rotation Rate=&MaskFloatBySpawnGroup(Masked Float=28.5, Passthrough Float=-25, Spawn Group Mask=$Particles.PartitionParticles.Partition))
   /SetVariables(Particles.ColorStars:LinearColor=&MultiplyLinearColor(
     Linear Color=<rgba 0.15 1 0.3625 0.5>
     Scale Factor=&MakeFloat4FromVectorAndFloat(
      "Vector (RGB)"=<1 1 1>
      "Float (Alpha)"=&Multiply_Float(A=1, B=&FNTimeOfDayScalar(Night Scale=0.75, Morning Scale=0.9, Evening Scale=0.9))
     )
    )) {isSetParameters=true}
   /SetVariables(Particles.StarsSize:Vector2f=&Multiply_Vector2D_ByFloat(Float=1.1, Vector2D=$Particles.SpriteSize)) {isSetParameters=true}
   &FN_Light_Attributes(Color=<rgba 0 10 0 1>, Create Color Attribute=true, Create Exponent Attribute=true, Create Radius Attribute=true, Exponent=1, Radius=60)
  ]
  rendererClasses [&NiagaraSpriteRendererProperties, &NiagaraLightRendererProperties]
  renderers [
   {rendererClass=&NiagaraSpriteRendererProperties}
   {rendererClass=&NiagaraSpriteRendererProperties, rendererIndex=1}
   {rendererClass=&NiagaraLightRendererProperties, rendererIndex=2}
  ]
 }}
summary {
 systemName=NS_NudgeHonk_HandGlow_01
 emitters [{
   emitterName=Hand_Glow_Loop
   rendererClasses [&NiagaraSpriteRendererProperties, &NiagaraLightRendererProperties]
  }]
 userVariables [
  User.bIsFrontEnd {type=NiagaraBool}
  User.bIsFrontEndPreview {type=NiagaraBool}
  User.bisHandHidden {type=NiagaraBool}
  User.bKillReaction {type=NiagaraBool}
  User.IsSkydiving {type=NiagaraBool}
 ]
}
